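# SoundCue for a single South Balk NPC alert-idle voice line, variant 07.
# One wave player plays its voice wave with boosted volume, routed through the critical enemy VO sound class.
# It uses the shared NPC alerted-VO attenuation and the boss combat chatter concurrency set to control falloff and overlap.
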SoundCue /SouthBulkAI/Audio/VO/SC/IP_SC_VO_SouthBalk_AlertIdle_07
root=WavePlayer_1
nodes [{id=WavePlayer_1, properties {soundWaveAssetPtr=/SouthBulkAI/Audio/VO/Waves/IP_VO_SouthBalk_AlertIdle_070}}]
properties {
 attenuationSettings=/Game/Sounds/Fort_Human_VO_BR/NPCs/Shared/NPC_VO_Alerted_Att
 maxDistance=12500
 soundClassObject=&Fort_VO_Enemy_Critical
 volumeMultiplier=2.5
 concurrencySet [&NPC_Boss_Combat_Chatter_Conc]
}
